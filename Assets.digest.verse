# Copyright Epic Games, Inc. All Rights Reserved.
#################################################
# Generated Digest of Verse API
# DO NOT modify this manually!
# Generated from build: ++Fortnite+Release-37.00-CL-44765404
#################################################


using {/Verse.org/Native}
using {/Verse.org/SceneGraph}
# Module import path: /playkigai.dig@fortnite.com/Minigame_Birdstar/EntityPrefabs
EntityPrefabs := module:
    @import_as("/Minigame_Birdstar/EntityPrefabs/bs_game_manager.bs_game_manager_C")
    bs_game_manager<scoped {/playkigai.dig@fortnite.com/Minigame_Birdstar}> := class<final>(entity):


    bs_game_manager_asset<scoped {/playkigai.dig@fortnite.com/Minigame_Birdstar}>:entity_prefab = external {}

    @import_as("/Minigame_Birdstar/EntityPrefabs/bs_game_side.bs_game_side_C")
    bs_game_side<scoped {/playkigai.dig@fortnite.com/Minigame_Birdstar}> := class<final>(entity):


    bs_game_side_asset<scoped {/playkigai.dig@fortnite.com/Minigame_Birdstar}>:entity_prefab = external {}

    @import_as("/Minigame_Birdstar/EntityPrefabs/bs_lever.bs_lever_C")
    bs_lever<scoped {/playkigai.dig@fortnite.com/Minigame_Birdstar}> := class<final>(entity):


    bs_lever_asset<scoped {/playkigai.dig@fortnite.com/Minigame_Birdstar}>:entity_prefab = external {}

    @import_as("/Minigame_Birdstar/EntityPrefabs/bs_note_circle.bs_note_circle_C")
    bs_note_circle<scoped {/playkigai.dig@fortnite.com/Minigame_Birdstar}> := class<final>(entity):


    bs_note_circle_asset<scoped {/playkigai.dig@fortnite.com/Minigame_Birdstar}>:entity_prefab = external {}

    @import_as("/Minigame_Birdstar/EntityPrefabs/bs_note_detector.bs_note_detector_C")
    bs_note_detector<scoped {/playkigai.dig@fortnite.com/Minigame_Birdstar}> := class<final>(entity):


    bs_note_detector_asset<scoped {/playkigai.dig@fortnite.com/Minigame_Birdstar}>:entity_prefab = external {}